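# Animation Blueprint for the female Armstrong 01 head that adds procedural secondary motion to hair bones.
# The input pose gets additive bone offsets and bone constraints that reshape the dyn_hair_02 chain.
# Separate Anim Dynamics nodes then spring-simulate each bone in that chain, followed by a looser, head-relative dyn_hair_01 with custom gravity.
# Copy Pose From Mesh and RigidBody nodes are unconnected, though initialization still caches the owner mesh as MeshToCopy.
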AnimBlueprint /Game/Characters/Player/Female/Medium/Heads/F_MED_CAU_Armstrong_Head_01/Mesh/F_MED_CAU_Armstrong_01_Dyn_AnimBP
target_skeleton=/Game/Characters/Player/Common/Fortnite_Base_Head/Fortnite_Base_Head_Export_Skeleton
graphs [
 AnimGraph {nodes [
   Copy Pose From Mesh(SourceMeshComponent=Get MeshToCopy() -> (MeshToCopy:Skeletal Mesh Component Object Reference))
   RigidBody() {alphaBoolBlend {blendOption=Linear}}
   AnimGraphNode_SubInput_0 = Input Pose()
   n1 = "Transform (Modify) Bone - Bone: dyn_hair_02"!(ComponentPose=Local To Component(LocalPose=@AnimGraphNode_SubInput_0), Alpha=1, Rotation="-20,0,-25") -> (Pose) {
    rotationMode=BMM_Additive
    rotationSpace=BCS_BoneSpace
    scaleMode=BMM_Additive
    scaleSpace=BCS_BoneSpace
    translationSpace=BCS_BoneSpace
    alphaBoolBlend {blendOption=Linear}
    alphaScaleBiasClamp {bInterpResult=true, interpSpeedDecreasing=100, interpSpeedIncreasing=100}
    boneToModify {boneName=dyn_hair_02}
    rotation {pitch=-20, roll=-25}
   }
   n2 = "Transform (Modify) Bone - Bone: dyn_hair_02_c"!(ComponentPose=@n1, Alpha=1, Scale="0.8,0.8,0.8") -> (Pose) {
    rotationSpace=BCS_BoneSpace
    scaleMode=BMM_Additive
    scaleSpace=BCS_BoneSpace
    translationSpace=BCS_BoneSpace
    alphaBoolBlend {blendOption=Linear}
    alphaScaleBiasClamp {bInterpResult=true, interpSpeedDecreasing=100, interpSpeedIncreasing=100}
    boneToModify {boneName=dyn_hair_02_c}
    scale=<0.8 0.8 0.8>
   }
   n3 = "Constraint - dyn_hair_02_a (P)"(ComponentPose=@n2, Alpha=1, ConstraintWeights_0=1) {
    alphaBoolBlend {blendOption=Linear}
    alphaScaleBiasClamp {bInterpResult=true, interpSpeedDecreasing=30, interpSpeedIncreasing=30}
    boneToModify {boneName=dyn_hair_02_a}
    constraintSetup [{targetBone {boneName=dyn_hair_02_d}}]
    constraintWeights [1]
   }
   n4 = "Constraint - dyn_hair_02_d (P)"(ComponentPose=@n3, Alpha=1, ConstraintWeights_0=1) {
    alphaBoolBlend {blendOption=Linear}
    alphaScaleBiasClamp {bInterpResult=true, interpSpeedDecreasing=30, interpSpeedIncreasing=30}
    boneToModify {boneName=dyn_hair_02_d}
    constraintSetup [{targetBone {boneName=dyn_hair_02_c}}]
    constraintWeights [1]
   }
   n5 = Anim Dynamics - Bone: dyn_hair_02(ComponentPose=@n4, Alpha=0.9) {
    alpha=0.9
    angularSpringConstant=1500
    bGravityOverrideInSimSpace=true
    numSolverIterationsPostUpdate=4
    numSolverIterationsPreUpdate=16
    simulationSpace=RootRelative
    alphaBoolBlend {blendOption=Linear}
    boundBone {boneName=dyn_hair_02}
    physicsBodyDefinitions [{
      sphereCollisionRadius=0
      boundBone {boneName=dyn_hair_02}
      boxExtents=<30 30 10>
      constraintSetup {angularTargetAxis=AxisZ, twistAxis=AxisZ, angularLimitsMax=<10 5 0>, angularLimitsMin=<-20 -20 0>, angularTarget=<0 0 1>}
      localJointOffset=<0 0 -3>
     }]
    relativeSpaceBone {boneName=pelvis}
   }
   n6 = Anim Dynamics - Bone: dyn_hair_02_a(ComponentPose=@n5, Alpha=0.9, ExternalForce="0,0,0") {
    alpha=0.9
    angularSpringConstant=1500
    bGravityOverrideInSimSpace=true
    numSolverIterationsPostUpdate=4
    numSolverIterationsPreUpdate=16
    alphaBoolBlend {blendOption=Linear}
    boundBone {boneName=dyn_hair_02_a}
    physicsBodyDefinitions [{
      sphereCollisionRadius=0
      boundBone {boneName=dyn_hair_02_a}
      boxExtents=<30 30 10>
      constraintSetup {angularTargetAxis=AxisZ, twistAxis=AxisZ, angularLimitsMax=<10 20 0>, angularLimitsMin=<-20 -20 0>, angularTarget=<0 0 1>}
      localJointOffset=<0 0 -3>
     }]
   }
   n7 = Anim Dynamics - Bone: dyn_hair_02_b(ComponentPose=@n6, Alpha=0.9, ExternalForce="0,0,0") {
    alpha=0.9
    angularSpringConstant=1500
    bGravityOverrideInSimSpace=true
    numSolverIterationsPostUpdate=4
    numSolverIterationsPreUpdate=16
    alphaBoolBlend {blendOption=Linear}
    boundBone {boneName=dyn_hair_02_b}
    physicsBodyDefinitions [{
      sphereCollisionRadius=0
      boundBone {boneName=dyn_hair_02_b}
      boxExtents=<30 30 10>
      constraintSetup {angularTargetAxis=AxisZ, twistAxis=AxisZ, angularLimitsMax=<10 20 0>, angularLimitsMin=<-20 -20 0>, angularTarget=<0 0 1>}
      localJointOffset=<0 0 -3>
     }]
   }
   n8 = Anim Dynamics - Bone: dyn_hair_02_c(ComponentPose=@n7, Alpha=0.9, ExternalForce="0,0,0") {
    alpha=0.9
    angularSpringConstant=1500
    bGravityOverrideInSimSpace=true
    numSolverIterationsPostUpdate=4
    numSolverIterationsPreUpdate=16
    alphaBoolBlend {blendOption=Linear}
    boundBone {boneName=dyn_hair_02_c}
    physicsBodyDefinitions [{
      sphereCollisionRadius=0
      boundBone {boneName=dyn_hair_02_c}
      boxExtents=<30 30 10>
      constraintSetup {angularTargetAxis=AxisZ, twistAxis=AxisZ, angularLimitsMax=<10 20 0>, angularLimitsMin=<-20 -20 0>, angularTarget=<0 0 1>}
      localJointOffset=<0 0 -3>
     }]
   }
   n9 = Anim Dynamics - Bone: dyn_hair_02_d(ComponentPose=@n8, Alpha=0.9, ExternalForce="0,0,0") {
    alpha=0.9
    angularSpringConstant=1500
    bGravityOverrideInSimSpace=true
    numSolverIterationsPostUpdate=4
    numSolverIterationsPreUpdate=16
    alphaBoolBlend {blendOption=Linear}
    boundBone {boneName=dyn_hair_02_d}
    physicsBodyDefinitions [{
      sphereCollisionRadius=0
      boundBone {boneName=dyn_hair_02_d}
      boxExtents=<30 30 10>
      constraintSetup {angularTargetAxis=AxisZ, twistAxis=AxisZ, angularLimitsMax=<10 20 0>, angularLimitsMin=<-20 -20 0>, angularTarget=<0 0 1>}
      localJointOffset=<0 0 -3>
     }]
   }
   n10 = Anim Dynamics - Bone: dyn_hair_01(ComponentPose=@n9, Alpha=0.2) {
    angularSpringConstant=1500
    bGravityOverrideInSimSpace=true
    bLinearSpring=true
    bOverrideLinearDamping=true
    bUseGravityOverride=true
    linearDampingOverride=0.99
    linearSpringConstant=250
    numSolverIterationsPostUpdate=4
    simulationSpace=BoneRelative
    alphaBoolBlend {blendOption=Linear}
    boundBone {boneName=dyn_hair_01}
    componentAppliedLinearAccClamp=<1250 1250 1250>
    componentLinearAccScale=<0.2 0.2 0.2>
    componentLinearVelScale=<1 1 1>
    gravityOverride {z=-100}
    physicsBodyDefinitions [{
      sphereCollisionRadius=0
      boundBone {boneName=dyn_hair_01}
      boxExtents=<13 13 13>
      constraintSetup {
       angularTargetAxis=AxisY
       twistAxis=AxisY
       angularLimitsMax=<5 5 5>
       angularLimitsMin=<-5 -5 -5>
       angularTarget=<0 0 1>
       linearAxesMax=<0.5 0.5 0.5>
       linearAxesMin=<-0.5 -0.5 -0.5>
      }
      localJointOffset=<0 0 0>
     }]
    relativeSpaceBone {boneName=head}
   }
   Output Pose(Result=Component To Local(ComponentPose=@n10)) {layerGroup=None}
  ]}
 EventGraph {nodes [
   Event BlueprintInitializeAnimation()
   > Set MeshToCopy(MeshToCopy:Skeletal Mesh Component Object Reference=GetOwnerSkeletalMeshComponent()) -> (Output_Get:Skeletal Mesh Component Object Reference)
   TryGetPawnOwner()
   Event BlueprintUpdateAnimation()
  ]}
]
variables [MeshToCopy {type=SkeletalMeshComponent}]
